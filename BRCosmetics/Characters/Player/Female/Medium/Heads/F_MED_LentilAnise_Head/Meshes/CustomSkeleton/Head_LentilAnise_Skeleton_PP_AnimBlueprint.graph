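# Post-process Anim Blueprint for the LentilAnise female head that overrides upper-eyelid motion during that character's hero emote.
# Each update checks whether the emote montage is active on the owner's anim instance and toggles the Custom Emote blend weight.
# A clamped remap of the Emote_upperLid_min_offset curve sets the blink sequence's explicit time.
# That evaluated pose layers onto the input pose from both upper-mid eyelid bones.
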AnimBlueprint /BRCosmetics/Characters/Player/Female/Medium/Heads/F_MED_LentilAnise_Head/Meshes/CustomSkeleton/Head_LentilAnise_Skeleton_PP_AnimBlueprint
target_skeleton=&/Fortnite_Base_Head_LentilAnise_Skeleton
graphs [
 AnimGraph {nodes [
   n1 = "Sequence Evaluator (by time) 'LentilAnise_Emote_Blink_Sequence'"!(ExplicitTime:float=Get Eye Emote Curve() -> (Eye Emote Curve:float)) -> (Pose:Pose Link Structure) {
    is=title AnimGraphNode_SequenceEvaluator
    sequence=&/LentilAnise_Emote_Blink_Sequence
   }
   n2 = Layered blend per bone(BasePose=Input Pose(), BlendPoses_0=@n1, BlendWeights_0=Get Custom Emote() -> (Custom Emote:double)) {layerSetup [{branchFilters [{boneName=L_eye_lid_upper_mid}, {boneName=R_eye_lid_upper_mid}]}]}
   Output Pose(Result=@n2)
  ]}
 EventGraph {nodes [
   Event BlueprintUpdateAnimation()
   > n1 = Sequence()
   > n2 = Branch(Condition=Montage_IsActive!(
     self:Anim Instance Object Reference=GetOwnerAnimInstance!(self:Custom Character Part Anim Instance Object Reference) -> (ReturnValue:Anim Instance Object Reference)
     Montage:Anim Montage Object Reference=/BRCosmetics/Animation/Game/MainPlayer/Emotes/LentilAnise/Hero/Emote_LentilAnise_CMF_Hero_M
    ) -> (ReturnValue:bool))
   > Set Custom Emote(Custom Emote:double=1) -> (Output_Get:double)
   Set Custom Emote(execute=@n2.else, Custom Emote=0)
   n3 = MapRangeClamped(Value=GetCurveValue(self=Self-Reference(), CurveName=Emote_upperLid_min_offset), InRangeA=-1, InRangeB=-0.5)
   Set Eye Emote Curve(execute=@n1.then_1, Eye Emote Curve:float=@n3) -> (Output_Get:float)
   MapRangeClamped(Value=@n3, InRangeA=0, InRangeB=1, OutRangeB=3.3)
   Event OnBeginPlayingEmote() {is=title K2Node_Event}
   Event OnEndPlayingEmote() {is=title K2Node_Event}
  ]}
]
variables [
 Custom Emote {type=number}
 Eye Emote Curve {type=number}
]
